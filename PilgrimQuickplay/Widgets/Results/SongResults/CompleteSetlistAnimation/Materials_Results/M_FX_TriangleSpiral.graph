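# Translucent UI material for scrolling triangle FX, filed under the song results screen's complete-setlist animation.
# Three channels of one triangle texture pan along X at different speeds, with per-layer scale and offset.
# SmoothStep thresholds with lerped bounds shape the layers, with reroutes declared for aspect-ratio UV setup and a powered horizontal gradient.
# Nested lerps feed an Opacity multiply; a Color parameter is also defined.
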Material /PilgrimQuickplay/Widgets/Results/SongResults/CompleteSetlistAnimation/Materials_Results/M_FX_TriangleSpiral
graph {
 properties {blendMode=BLEND_Translucent, materialDomain=MD_UI, pixelDepthOffsetMode=PDOM_Legacy, usedShadingModels=Default Lit}
 nodes [
  n1 = MaterialFunctionCall() {materialFunction=&GetUserInterfaceUV}
  NamedRerouteDeclaration(None=@n1.Normalized UV) {name=NormalizedUV, nodeColor=<rgba 1 0 0.8906 1>}
  n2 = MaterialFunctionCall(Float2=@n1.Pixel Size) {materialFunction=&BreakOutFloat2Components}
  n3 = Max(B=@n2.G) {constA=0.1}
  NamedRerouteDeclaration(None=AppendVector(A=Divide(A=@n2.R, B=@n3), B=Constant() {r=1})) {name=UVSetup, nodeColor=<rgba 0.8516 1 0 1>}
  n4 = TextureObjectParameter() {parameterName=Texture, samplerType=SAMPLERTYPE_LinearColor, texture=/Game/UI/Foundation/Items/Textures/T-Card-Hero-Anim-Triangle}
  NamedRerouteDeclaration(None=@n4) {name=TriangleTexture, nodeColor=<rgba 0.8281 0 1 1>}
  n5 = TextureCoordinate()
  NamedRerouteDeclaration(None=Power(Base=ComponentMask(None=@n5) {r=true}) {constExponent=1.7}) {name=TexCoordPower, nodeColor=<rgba 0 1 0.5078 1>}
  NamedRerouteDeclaration(None=ScalarParameter() {defaultValue=1.95, parameterName=ss min}) {name=SS Min, nodeColor=<rgba 1 0.1875 0 1>}
  NamedRerouteDeclaration(None=ScalarParameter() {defaultValue=0.35, parameterName=ss max}) {name=SS Max, nodeColor=<rgba 0 0.1328 1 1>}
  n6 = NamedRerouteUsage()
  n7 = Multiply(A=NamedRerouteUsage(), B=ScalarParameter() {defaultValue=0.5743, parameterName=tri-L2-Scale})
  n8 = Constant()
  n9 = ScalarParameter() {defaultValue=-0.6765, parameterName=tri-L2-Offset}
  n10 = Panner(Coordinate=Add(A=Multiply(A=@n6, B=@n7), B=AppendVector(A=@n8, B=@n9))) {speedX=-0.45}
  n11 = NamedRerouteUsage()
  n12 = NamedRerouteUsage()
  n13 = NamedRerouteUsage()
  n14 = NamedRerouteUsage()
  n15 = NamedRerouteUsage()
  NamedRerouteDeclaration(None=MaterialFunctionCall(
    Alpha=TextureSample(UVs=@n10, Tex=@n11) {samplerType=SAMPLERTYPE_LinearColor, texture=&T-Card-Hero-Anim-Triangle}.G
    Min=LinearInterpolate(B=@n12, Alpha=@n13) {constA=0.25, constB=0.9}
    Max=LinearInterpolate(A=@n14, Alpha=@n15) {constA=0.1, constB=2}
   ) {materialFunction=&SmoothStep}.Result) {name=TrianglesGreen, nodeColor=<rgba 1 0 0.375 1>}
  n16 = NamedRerouteUsage()
  n17 = ScalarParameter() {defaultValue=0.33, parameterName=tri-L1-Scale}
  n18 = Add(
   A=Multiply(A=NamedRerouteUsage(), B=Multiply(A=@n16, B=@n17))
   B=AppendVector(A=Constant(), B=ScalarParameter() {defaultValue=0.2476, parameterName=tri-L1-Offset})
  )
  NamedRerouteDeclaration(None=TextureSample(UVs=Panner(Coordinate=@n18) {speedX=-0.5}, Tex=NamedRerouteUsage()) {samplerType=SAMPLERTYPE_LinearColor, texture=&T-Card-Hero-Anim-Triangle}.R) {name=TrianglesRed, nodeColor=<rgba 0 0.6953 1 1>}
  n19 = NamedRerouteUsage()
  n20 = Multiply(A=NamedRerouteUsage(), B=ScalarParameter() {defaultValue=0.9048, parameterName=tri-L3-Scale})
  n21 = Constant()
  n22 = ScalarParameter() {defaultValue=0.019, parameterName=tri-L3-Offset}
  n23 = Panner(Coordinate=Add(A=Multiply(A=@n19, B=@n20), B=AppendVector(A=@n21, B=@n22))) {speedX=-0.35}
  n24 = NamedRerouteUsage()
  n25 = NamedRerouteUsage()
  n26 = NamedRerouteUsage()
  n27 = NamedRerouteUsage()
  n28 = Multiply(A=NamedRerouteUsage())
  NamedRerouteDeclaration(None=MaterialFunctionCall(
    Alpha=TextureSample(UVs=@n23, Tex=@n24) {samplerType=SAMPLERTYPE_LinearColor, texture=&T-Card-Hero-Anim-Triangle}.B
    Min=LinearInterpolate(B=@n25, Alpha=@n26) {constA=0.25, constB=0.8}
    Max=LinearInterpolate(A=@n27, Alpha=@n28) {constA=0.2, constB=2}
   ) {materialFunction=&SmoothStep}.Result) {name=TrianglesBlue, nodeColor=<rgba 1 0.9844 0 1>}
  n29 = LinearInterpolate(A=NamedRerouteUsage(), Alpha=NamedRerouteUsage()) {constB=0.25}
  n30 = MaterialFunctionCall(
   Alpha=NamedRerouteUsage()
   Min=LinearInterpolate(B=NamedRerouteUsage(), Alpha=NamedRerouteUsage()) {constA=0.25, constB=0.9}
   Max=LinearInterpolate(A=NamedRerouteUsage(), Alpha=NamedRerouteUsage()) {constA=0.1, constB=2}
  ) {materialFunction=&SmoothStep}
  Reroute(None=Multiply(A=LinearInterpolate(A=@n29, Alpha=@n30.Result), B=ScalarParameter() {defaultValue=1, parameterName=Opacity}))
  VectorParameter() {group=Color, parameterName=Color, defaultValue=<rgba 1 0 0.9675 0.302>}
 ]
}
