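# Translucent UI material for the animated cloud background in the Pilgrim Quickplay complete-setlist results animation.
# Horizontally scrolling noise multiplies an oppositely scrolling, flow-map-distorted cloud texture, then a sine-pulsed threshold varies its density.
# An eased vertical gradient set by Smoke Height masks the clouds.
# That mask blends background and cloud colors with an additive boost, and interpolates between separate background and cloud opacities.
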Material /PilgrimQuickplay/Widgets/Results/SongResults/CompleteSetlistAnimation/Materials_Results/M_UI_Background
graph {
 properties {blendMode=BLEND_Translucent, materialDomain=MD_UI, pixelDepthOffsetMode=PDOM_Legacy, usedShadingModels=Default Lit}
 nodes [
  n1 = VectorParameter() {parameterName=Color Clouds, defaultValue=<rgba 0.3753 0.1277 1 1>}
  n2 = MaterialFunctionCall(UV=MaterialFunctionCall() {materialFunction=&GetUserInterfaceUV}.Normalized UV) {materialFunction=&MF_UI_UVScaling}
  n3 = Time()
  n4 = ScalarParameter() {defaultValue=0.1, parameterName=Clouds Speed}
  n5 = AppendVector(A=Frac(None=Multiply(A=@n3, B=@n4) {constB=0.1}), B=Constant())
  n6 = Add(A=@n2.ScaleX, B=@n5)
  n7 = TextureObject() {samplerType=SAMPLERTYPE_Grayscale, texture=/FortWater/Textures/LowResBlurredNoise}
  n8 = Subtract(A=@n2.ScaleX, B=@n5)
  n9 = TextureObjectParameter() {parameterName=Cloud Texture}
  n10 = TextureSample(UVs=@n6, Tex=@n7)
  n11 = Constant() {r=0.1}
  n12 = Multiply(
   A=LinearInterpolate(Alpha=TextureSample(UVs=@n6, Tex=@n7) {samplerSource=SSM_Wrap_WorldGroupSettings}.R) {constA=0.75}
   B=ComponentMask(None=MaterialFunctionCall!(UVs=@n8, Texture Object=@n9, Grayscale Flow Map=@n10.RGB, Flow Direction, Flow Intensity=@n11, Flow Speed=@n4) {materialFunction=/FortUILibrary/MaterialFunctions/Effects/MF_UI_GrayscaleFlowMap_New}.Distorted Texture) {r=true}
  )
  n13 = Power(Base=@n12)
  n14 = Sine(None=MaterialFunctionCall!(Animation Duration=ScalarParameter() {defaultValue=4, parameterName=Duration}, Delay=Constant(), Reset to Default) {materialFunction=/FortUILibrary/MaterialFunctions/Time/MF_UI_AnimationDelay}.Result)
  n15 = LinearInterpolate(Alpha=Add(A=Multiply(A=@n14) {constB=0.5}) {constB=0.5}) {constB=0.25}
  MaterialExpressionRemap_2 = MaterialXRemap!(None=Saturate(None=Subtract(A=@n13, B=@n15)), InputLow, InputHigh=OneMinus(None=@n15), TargetLow, TargetHigh)
  n16 = ComponentMask(None=@n2.Scaled to Fit) {g=true}
  n17 = ScalarParameter() {defaultValue=1, parameterName=Smoke Height, sliderMax=1}
  n18 = OneMinus(None=@n17)
  MaterialExpressionRemap_0 = MaterialXRemap(None=Saturate(None=Subtract(A=@n16, B=@n18) {constB=0.5}), InputHigh=@n17) {inputHighDefault=0.5}
  n19 = Multiply(
   A=Saturate(None=Add(A=@MaterialExpressionRemap_2, B=@n12))
   B=MaterialFunctionCall("Value (0-1)"=@MaterialExpressionRemap_0) {materialFunction=&MF_UI_EaseCurves}.EaseInQuint
  )
  n20 = ScalarParameter() {defaultValue=1, parameterName=Clouds Opacity, sliderMax=1}
  n21 = Multiply(A=@n19, B=@n20)
  Add(
   A=LinearInterpolate(A=VectorParameter() {parameterName=Color BG, defaultValue=<rgba 0 0.0168 0.4563 1>}.RGB, B=@n1.RGB, Alpha=@n21)
   B=Multiply(A=@n1.RGB, B=@n21)
  )
  LinearInterpolate(A=ScalarParameter() {defaultValue=1, parameterName=Bg Opacity, sliderMax=1}, B=@n20, Alpha=@n19)
  Saturate(None=@n16)
 ]
}
