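# Additive UI glow material for the complete-setlist animation on the Pilgrim Quickplay song results screen.
# A radial gradient multiplied by a rotated linear gradient is sine-eased and scaled by Opacity to form the glow.
# An aspect-corrected, antialiased hex grid, scaled by HexGridOpacity, is clipped to the glow with a Min, boosted and added back onto it.
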Material /PilgrimQuickplay/Widgets/Results/SongResults/CompleteSetlistAnimation/Materials_Results/M_UI_Glow
graph {
 properties {blendMode=BLEND_Additive, materialDomain=MD_UI, pixelDepthOffsetMode=PDOM_Legacy, usedShadingModels=Default Lit}
 nodes [
  n1 = MaterialFunctionCall() {materialFunction=&MF_UI_RadialGradient}
  n2 = MaterialFunctionCall(RotationAngle=Constant() {r=0.375}) {materialFunction=&MF_UI_LinearGradient}
  n3 = Multiply(
   A=MaterialFunctionCall("Value (0-1)"=Multiply(A=@n1.Smooth, B=@n2.Result)) {materialFunction=&MF_UI_EaseCurves}.EaseInSine
   B=ScalarParameter() {defaultValue=1, parameterName=Opacity, sliderMax=1}
  )
  n4 = TextureCoordinate()
  n5 = MaterialFunctionCall!(Dimensions) {materialFunction=/FortUILibrary/MaterialFunctions/Basics/MF_UI_GetAspectRatio}
  n6 = ComponentMask(None=MaterialFunctionCall!(UVCoords=Multiply(A=@n4, B=@n5.H Ratio), HexSize=Constant() {r=0.03}, HexGradientSharpness=Constant() {r=1}) {materialFunction=/Game/Athena/Asteria/Environments/Landscape/Materials/MaterialFunctions/UtilityFunctions/HexGrid}.GradientsRGBA) {r=true}
  n7 = Multiply(
   A=MaterialFunctionCall(DistanceField=@n6, Threshold=Constant() {r=0.14}) {materialFunction=&MF_UI_AntiAliasing}.Inner
   B=ScalarParameter() {parameterName=HexGridOpacity}
  )
  Add(A=@n3, B=Multiply(A=Min(A=@n3, B=@n7)) {constB=1.5})
  Multiply(A=MaterialFunctionCall("Value (0-1)"=@n2.Result) {materialFunction=&MF_UI_EaseCurves}.EaseInCubic)
  Subtract(A=@n6) {constB=0.06}
  VectorParameter() {parameterName=Glow Color, defaultValue=<rgba 1 1 1 1>}
 ]
}
