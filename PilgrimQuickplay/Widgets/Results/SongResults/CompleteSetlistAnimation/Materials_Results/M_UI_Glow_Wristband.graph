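# Additive UI glow material for the wristband in the song results setlist-completion animation, with a Color parameter.
# An exponential radial gradient over 9-slice UI UVs, scaled by XScale and YScale, provides the glow shape with adjustable center, radius and density.
# The gradient is multiplied by an Opacity parameter and a rerouted pulse value.
# A static switch selects either constant one or a time-driven LinearSine lerp using PulseConstraints, with speed and offset controls.
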Material /PilgrimQuickplay/Widgets/Results/SongResults/CompleteSetlistAnimation/Materials_Results/M_UI_Glow_Wristband
graph {
 properties {blendMode=BLEND_Additive, materialDomain=MD_UI, pixelDepthOffsetMode=PDOM_Legacy, usedShadingModels=Default Lit}
 nodes [
  n1 = MaterialFunctionCall!(
   Texture Scale=AppendVector(
    A=ScalarParameter() {defaultValue=0.8, group=Scale, parameterName=XScale}
    B=ScalarParameter() {defaultValue=0.8, group=Scale, parameterName=YScale}
   )
   UVs=MaterialFunctionCall() {materialFunction=&GetUserInterfaceUV}.9-Slice UV
   Offset XY Texturesize Z
  ) {materialFunction=/Engine/Functions/Engine_MaterialFunctions01/Texturing/ScaleUVsAroundPoint}
  n2 = VectorParameter() {parameterName=Center, defaultValue=<rgba 0.5 0.5 0 1>}
  n3 = AppendVector(A=@n2.R, B=@n2.R)
  n4 = StaticBoolParameter() {defaultValue=true, parameterName=InvertDensity}
  n5 = ScalarParameter() {defaultValue=2.3, parameterName=Density}
  n6 = ScalarParameter() {defaultValue=0.5, parameterName=Radius}
  Multiply(
   A=Multiply(
    A=NamedRerouteUsage()
    B=MaterialFunctionCall(UVs=@n1.UVs, CenterPosition=@n3, Radius=@n6, Density=@n5, Invert Density=@n4) {materialFunction=&RadialGradientExponential}.RadialGradientExponential
   )
   B=ScalarParameter() {defaultValue=1, parameterName=Opacity}
  )
  n7 = VectorParameter() {group=Pulse, parameterName=PulseConstraints, defaultValue=<rgba 0 1 0 1>}
  n8 = Multiply(A=Time(), B=ScalarParameter() {defaultValue=1, group=Pulse, parameterName=PulseSpeed})
  n9 = Constant() {r=1}
  n10 = MaterialFunctionCall(Value=Add(A=Fmod(A=@n8, B=@n9), B=ScalarParameter() {parameterName=PulseTimeOffset})) {materialFunction=&LinearSine}
  NamedRerouteDeclaration(None=StaticSwitchParameter(True=LinearInterpolate(A=@n7.R, B=@n7.R, Alpha=@n10.Linear Sine), False=Constant() {r=1}) {group=Pulse, parameterName=isPulsing}) {name=PulsingAnimation, nodeColor=<rgba 1 0.0703 0 1>}
  VectorParameter() {parameterName=Color, defaultValue=<rgba 0.0296 1 0.5972 1>}
 ]
}
