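# Translucent UI material for the song results complete-setlist animation, drawing a three-color gradient with tapered alpha.
# Static switches select an exponential radial gradient, a rotatable linear fade, or a beam that multiplies the two through a lerp.
# The chosen mask is scaled by an animatable start-to-stop alpha blend.
# It then drives 3ColorBlend and a separate opacity chain, each with its own power taper.
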Material /PilgrimQuickplay/Widgets/Results/SongResults/CompleteSetlistAnimation/Materials_Results/M_UI_GradientHorizontal
graph {
 properties {blendMode=BLEND_Translucent, materialDomain=MD_UI, pixelDepthOffsetMode=PDOM_Legacy, usedShadingModels=Default Lit}
 nodes [
  n1 = ScalarParameter() {defaultValue=0.5, parameterName=ScaleRadialX}
  n2 = Constant() {r=1}
  n3 = MaterialFunctionCall(
   UVs=Multiply(A=TextureCoordinate(), B=AppendVector(A=@n1, B=@n2))
   CenterPosition=AppendVector(A=ScalarParameter() {parameterName=OffsetRadialX}, B=Constant() {r=0.5})
  ) {materialFunction=&RadialGradientExponential}
  n4 = Add(A=TextureCoordinate(), B=ScalarParameter() {parameterName=OffsetFade})
  n5 = ComponentMask(None=MaterialFunctionCall(UVs=OneMinus(None=@n4), "Rotation Angle (0-1)"=ScalarParameter() {parameterName=FadeRotation}) {materialFunction=&CustomRotator}.Rotated Values) {r=true}
  n6 = LinearInterpolate(B=@n3.RadialGradientExponential, Alpha=@n5)
  n7 = Multiply(
   A=StaticSwitchParameter(True=@n3.RadialGradientExponential, False=StaticSwitchParameter(True=@n6, False=@n5) {parameterName=IsBeam}) {parameterName=IsRadialGradient}
   B=LinearInterpolate(
    A=ScalarParameter() {defaultValue=1, parameterName=Alpha_Start}
    B=ScalarParameter() {defaultValue=1, parameterName=Alpha_Stop}
    Alpha=ScalarParameter() {parameterName=Animate_Alpha}
   )
  )
  MaterialFunctionCall(
   A=VectorParameter() {parameterName=ColorFalloff, defaultValue=<rgba 1 0 0 1>}.RGB
   B=VectorParameter() {parameterName=ColorMiddle, defaultValue=<rgba 0 1 0 1>}.RGB
   C=VectorParameter() {parameterName=ColorHotspot, defaultValue=<rgba 0 0 1 1>}.RGB
   Alpha=Power(Base=@n7, Exp=ScalarParameter() {defaultValue=1.22, parameterName=TaperColorGradient})
  ) {materialFunction=&3ColorBlend}
  n8 = ScalarParameter() {defaultValue=1, parameterName=MultiplyGradient}
  n9 = Power(Base=Saturate(None=Multiply(A=@n7, B=@n8)), Exp=ScalarParameter() {defaultValue=1.22, parameterName=TaperAlphaGradient})
  Saturate(None=Multiply(A=Saturate(None=@n9), B=ScalarParameter() {defaultValue=1, parameterName=FinalAlpha}))
 ]
}
