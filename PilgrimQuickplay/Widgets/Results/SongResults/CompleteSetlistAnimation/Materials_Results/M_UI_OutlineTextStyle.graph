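# UI-domain additive font material, named as an outline text style for the complete-setlist animation on the Quickplay song results screen.
# Font signed distance feeds MF_UI_AntiAliasing with a low constant threshold and smoothing.
# Its Inner mask is multiplied by implicit font opacity, vertex alpha and an Opacity parameter.
# A Color parameter feeds both lerp inputs, so the vertex-color alpha input leaves the color unchanged.
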Material /PilgrimQuickplay/Widgets/Results/SongResults/CompleteSetlistAnimation/Materials_Results/M_UI_OutlineTextStyle
graph {
 properties {bAutomaticallySetUsageInEditor=false, blendMode=BLEND_Additive, materialDomain=MD_UI, usedShadingModels=Default Lit}
 nodes [
  n1 = VectorParameter() {parameterName=Color, defaultValue=<rgba 1 1 1 1>}
  n2 = VertexColor()
  LinearInterpolate(A=@n1.RGB, B=@n1.RGB, Alpha=@n2)
  n3 = FontSignedDistance()
  n4 = MaterialFunctionCall(DistanceField=@n3.Signed Distance, Threshold=Constant() {r=0.02}, Smoothing=Constant() {r=0.01}) {materialFunction=&MF_UI_AntiAliasing}
  Multiply(A=Multiply(A=@n2.A, B=Multiply(A=@n4.Inner, B=@n3.Implicit Opacity)), B=ScalarParameter() {defaultValue=1, parameterName=Opacity})
 ]
}
